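# Physics asset for the medium male Hardcore Sportz body cosmetic, covering jointed secondary motion of knee flaps, calf flaps and sleeves rather than a full-body ragdoll.
# Each knee and calf flap capsule is constrained to its own tiny parent-bone capsule, with moderate swing and very little twist.
# Each sleeve is a two-link chain hung from a small upper-arm twist body, with tighter swing at the first link than the second.
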PhysicsAsset /BRCosmetics/Characters/Player/Male/Medium/Bodies/M_MED_Hardcore_Sportz/Meshes/M_MED_Hardcore_Sportz_Physics
bodies {
 dyn_calf_flapA_in_l [{shapeType=Capsule, length=3, radius=2.25, center=<1 0 0>, rotation=<rot 55.4612 179.9988 179.9987>}]
 dyn_calf_flapA_in_r [{shapeType=Capsule, length=0.5, radius=3.5, center=<-0.95 0 0>, rotation=<rot 90 0 0>}]
 dyn_calf_flapA_out_l [{shapeType=Capsule, length=3, radius=2.75, center=<1 0 0>, rotation=<rot 90 0 0>}]
 dyn_calf_flapA_out_r [{shapeType=Capsule, length=2.25, radius=4, center=<-0.8 0 0>, rotation=<rot 90 0 0>}]
 dyn_calf_flapB_in_l [{shapeType=Capsule, length=1.25, radius=1.5, center=<1 0 0>, rotation=<rot 80.7495 179.9985 179.9985>}]
 dyn_calf_flapB_in_r [{shapeType=Capsule, length=1.5, radius=3.25, center=<-0.77 0 0>, rotation=<rot 90 0 0>}]
 dyn_calf_flapB_out_l [{shapeType=Capsule, length=2.25, radius=4, center=<1 0 0>, rotation=<rot 90 0 0>}]
 dyn_calf_flapB_out_r [{shapeType=Capsule, length=2.25, radius=4, center=<-1.2 0 0>, rotation=<rot 90 0 0>}]
 dyn_knee_flap_inB_l [{shapeType=Capsule, length=3.75, radius=2.5, center=<0.9 0 0>, rotation=<rot 90 0 0>}]
 dyn_knee_flap_inB_r [{shapeType=Capsule, length=2.25, radius=4, center=<1.25 0 0>, rotation=<rot 90 0 0>}]
 dyn_knee_flap_in_l [{shapeType=Capsule, length=2.25, radius=4, center=<0.35 0 0>, rotation=<rot 90 0 0>}]
 dyn_knee_flap_in_r [{shapeType=Capsule, length=2, radius=2.25, center=<0.75 0 0>, rotation=<rot 90 0 0>}]
 dyn_knee_flap_outB_l [{shapeType=Capsule, length=3.25, radius=3, center=<1.5 0 0>, rotation=<rot 90 0 0>}]
 dyn_knee_flap_outB_r [{shapeType=Capsule, length=2.25, radius=4, center=<-1.5 0 0>, rotation=<rot 90 0 0>}]
 dyn_knee_flap_out_l [{shapeType=Capsule, length=1.5, radius=2.25, center=<0.8 0 0>, rotation=<rot 90 0 0>}]
 dyn_knee_flap_out_r [{shapeType=Capsule, length=2.25, radius=2.75, center=<-0.7 0 0>, rotation=<rot 90 0 0>}]
 dyn_parent_calf_flapA_in_l [{shapeType=Capsule, length=0.5, radius=0.15}]
 dyn_parent_calf_flapA_in_r [{shapeType=Capsule, length=0.5, radius=0.15}]
 dyn_parent_calf_flapA_out_l [{shapeType=Capsule, length=0.5, radius=0.15}]
 dyn_parent_calf_flapA_out_r [{shapeType=Capsule, length=0.5, radius=0.15}]
 dyn_parent_calf_flapB_in_l [{shapeType=Capsule, length=0.5, radius=0.15}]
 dyn_parent_calf_flapB_in_r [{shapeType=Capsule, length=0.5, radius=0.15}]
 dyn_parent_calf_flapB_out_l [{shapeType=Capsule, length=0.5, radius=0.15}]
 dyn_parent_calf_flapB_out_r [{shapeType=Capsule, length=0.5, radius=0.15}]
 dyn_parent_knee_strap_inB_l [{shapeType=Capsule, length=0.5, radius=0.15}]
 dyn_parent_knee_strap_inB_r [{shapeType=Capsule, length=0.5, radius=0.15}]
 dyn_parent_knee_strap_in_l [{shapeType=Capsule, length=0.5, radius=0.15}]
 dyn_parent_knee_strap_in_r [{shapeType=Capsule, length=0.5, radius=0.15}]
 dyn_parent_knee_strap_outB_l [{shapeType=Capsule, length=0.5, radius=0.15}]
 dyn_parent_knee_strap_outB_r [{shapeType=Capsule, length=0.5, radius=0.15}]
 dyn_parent_knee_strap_out_l [{shapeType=Capsule, length=0.5, radius=0.15}]
 dyn_parent_knee_strap_out_r [{shapeType=Capsule, length=0.5, radius=0.15}]
 dyn_sleeve_01_l [{shapeType=Capsule, length=3.25, radius=6.5, center=<0.25 0 0>, rotation=<rot 90 0 0>}]
 dyn_sleeve_01_r [{shapeType=Capsule, length=3.25, radius=6.5, center=<-0.25 0 0>, rotation=<rot 90 0 0>}]
 dyn_sleeve_02_l [{shapeType=Capsule, length=3, radius=4.5, center=<0.4 0 0>, rotation=<rot 90 0 0>}]
 dyn_sleeve_02_r [{shapeType=Capsule, length=3, radius=4.5, center=<-0.4 0 0>, rotation=<rot 90 0 0>}]
 upperarm_twist_02_l [{shapeType=Capsule, length=1, radius=1}]
 upperarm_twist_02_r [{shapeType=Capsule, length=1, radius=1}]
}
constraints [
 {bone1Name=dyn_knee_flap_in_l, bone2Name=dyn_parent_knee_strap_in_l, swing1LimitDegrees=25, swing2LimitDegrees=15, twistLimitDegrees=3}
 {bone1Name=dyn_knee_flap_out_l, bone2Name=dyn_parent_knee_strap_out_l, swing1LimitDegrees=25, swing2LimitDegrees=25, twistLimitDegrees=3}
 {bone1Name=dyn_knee_flap_inB_l, bone2Name=dyn_parent_knee_strap_inB_l, swing1LimitDegrees=25, swing2LimitDegrees=25, twistLimitDegrees=3}
 {bone1Name=dyn_knee_flap_outB_l, bone2Name=dyn_parent_knee_strap_outB_l, swing1LimitDegrees=25, swing2LimitDegrees=15, twistLimitDegrees=3}
 {bone1Name=dyn_calf_flapA_in_l, bone2Name=dyn_parent_calf_flapA_in_l, swing1LimitDegrees=25, swing2LimitDegrees=15, twistLimitDegrees=3}
 {bone1Name=dyn_calf_flapA_out_l, bone2Name=dyn_parent_calf_flapA_out_l, swing1LimitDegrees=25, swing2LimitDegrees=25, twistLimitDegrees=3}
 {bone1Name=dyn_calf_flapB_in_l, bone2Name=dyn_parent_calf_flapB_in_l, swing1LimitDegrees=25, swing2LimitDegrees=25, twistLimitDegrees=3}
 {bone1Name=dyn_calf_flapB_out_l, bone2Name=dyn_parent_calf_flapB_out_l, swing1LimitDegrees=25, swing2LimitDegrees=15, twistLimitDegrees=3}
 {bone1Name=dyn_knee_flap_in_r, bone2Name=dyn_parent_knee_strap_in_r, swing1LimitDegrees=25, swing2LimitDegrees=25, twistLimitDegrees=3}
 {bone1Name=dyn_knee_flap_out_r, bone2Name=dyn_parent_knee_strap_out_r, swing1LimitDegrees=25, swing2LimitDegrees=15, twistLimitDegrees=3}
 {bone1Name=dyn_knee_flap_inB_r, bone2Name=dyn_parent_knee_strap_inB_r, swing1LimitDegrees=25, swing2LimitDegrees=25, twistLimitDegrees=3}
 {bone1Name=dyn_knee_flap_outB_r, bone2Name=dyn_parent_knee_strap_outB_r, swing1LimitDegrees=25, swing2LimitDegrees=15, twistLimitDegrees=3}
 {bone1Name=dyn_calf_flapA_in_r, bone2Name=dyn_parent_calf_flapA_in_r, swing1LimitDegrees=25, swing2LimitDegrees=25, twistLimitDegrees=3}
 {bone1Name=dyn_calf_flapA_out_r, bone2Name=dyn_parent_calf_flapA_out_r, swing1LimitDegrees=25, swing2LimitDegrees=15, twistLimitDegrees=3}
 {bone1Name=dyn_calf_flapB_in_r, bone2Name=dyn_parent_calf_flapB_in_r, swing1LimitDegrees=25, swing2LimitDegrees=15, twistLimitDegrees=3}
 {bone1Name=dyn_calf_flapB_out_r, bone2Name=dyn_parent_calf_flapB_out_r, swing1LimitDegrees=25, swing2LimitDegrees=25, twistLimitDegrees=3}
 {bone1Name=dyn_sleeve_01_l, bone2Name=upperarm_twist_02_l, swing1LimitDegrees=5, swing2LimitDegrees=5, twistLimitDegrees=5}
 {bone1Name=dyn_sleeve_02_l, bone2Name=dyn_sleeve_01_l, swing1LimitDegrees=11, swing2LimitDegrees=11, twistLimitDegrees=5}
 {bone1Name=dyn_sleeve_01_r, bone2Name=upperarm_twist_02_r, swing1LimitDegrees=5, swing2LimitDegrees=5, twistLimitDegrees=5}
 {bone1Name=dyn_sleeve_02_r, bone2Name=dyn_sleeve_01_r, swing1LimitDegrees=11, swing2LimitDegrees=11, twistLimitDegrees=5}
]
