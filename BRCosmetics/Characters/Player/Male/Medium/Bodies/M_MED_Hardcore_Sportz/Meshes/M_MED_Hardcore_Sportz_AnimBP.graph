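# Cosmetic AnimBlueprint for the M_MED_Hardcore_Sportz player body, adding secondary motion to a pose copied from the attached parent mesh.
# Event graph turns arm and knee twist deltas into alphas driving additive left shoulder pad, knee and calf-squish bone offsets.
# Weighted constraints position knee straps, then the cached pose feeds two RigidBody simulations, switching from root-based to partial-alpha spine_04-based while moving or accelerating.
# The calf-flap constraint chain is unconnected to the output.
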AnimBlueprint /BRCosmetics/Characters/Player/Male/Medium/Bodies/M_MED_Hardcore_Sportz/Meshes/M_MED_Hardcore_Sportz_AnimBP
target_skeleton=&/M_MED_Hardcore_Sportz_Skeleton
graphs [
 AnimGraph {nodes [
   n1 = "Constraint - dyn_parent_calf_flapA_in_l (PPPPP)"!(
    ComponentPose
    Alpha=1
    ConstraintWeights_0=0.2069
    ConstraintWeights_1:float=0.0017
    ConstraintWeights_2:float=0.194
    ConstraintWeights_3:float=0.0068
    ConstraintWeights_4:float=0.5906
   ) -> (Pose) {
    alphaBoolBlend {blendOption=Linear}
    boneToModify {boneName=dyn_parent_calf_flapA_in_l}
    constraintSetup [
     {targetBone {boneName=calf_l}}
     {targetBone {boneName=calf_twist_01_l}}
     {targetBone {boneName=calf_twist_02_l}}
     {targetBone {boneName=thigh_l}}
     {targetBone {boneName=dyn_calf_flapA_in_l}}
    ]
    constraintWeights [1,1,1,1,1]
   }
   n2 = "Constraint - dyn_parent_calf_flapA_out_l (PPPPP)"!(
    ComponentPose=@n1
    Alpha=1
    ConstraintWeights_0=0.2298
    ConstraintWeights_1:float=0.0006
    ConstraintWeights_2:float=0.2884
    ConstraintWeights_3:float=0.0003
    ConstraintWeights_4:float=0.4808
   ) -> (Pose) {
    alphaBoolBlend {blendOption=Linear}
    boneToModify {boneName=dyn_parent_calf_flapA_out_l}
    constraintSetup [
     {targetBone {boneName=calf_l}}
     {targetBone {boneName=calf_twist_01_l}}
     {targetBone {boneName=calf_twist_02_l}}
     {targetBone {boneName=thigh_l}}
     {targetBone {boneName=dyn_calf_flapA_out_l}}
    ]
    constraintWeights [1,1,1,1,1]
   }
   n3 = "Constraint - dyn_parent_calf_flapB_in_l (PPPPP)"!(
    ComponentPose=@n2
    Alpha=1
    ConstraintWeights_0=0.0213
    ConstraintWeights_1:float=0.1288
    ConstraintWeights_2:float=0
    ConstraintWeights_3:float=0.2738
    ConstraintWeights_4:float=0.5761
   ) -> (Pose) {
    alphaBoolBlend {blendOption=Linear}
    boneToModify {boneName=dyn_parent_calf_flapB_in_l}
    constraintSetup [
     {targetBone {boneName=calf_l}}
     {targetBone {boneName=calf_twist_01_l}}
     {targetBone {boneName=foot_l}}
     {targetBone {boneName=calf_twist_02_l}}
     {targetBone {boneName=dyn_calf_flapB_in_l}}
    ]
    constraintWeights [1,1,1,1,1]
   }
   n4 = "Constraint - dyn_parent_calf_flapB_out_l (PPPPP)"!(
    ComponentPose=@n3
    Alpha=1
    ConstraintWeights_0=0.0445
    ConstraintWeights_1:float=0.1741
    ConstraintWeights_2:float=0.378
    ConstraintWeights_3:float=0.3548
    ConstraintWeights_4:float=0.0485
   ) -> (Pose) {
    alphaBoolBlend {blendOption=Linear}
    boneToModify {boneName=dyn_parent_calf_flapB_out_l}
    constraintSetup [
     {targetBone {boneName=calf_l}}
     {targetBone {boneName=calf_twist_01_l}}
     {targetBone {boneName=dyn_calf_flapB_out_l}}
     {targetBone {boneName=calf_twist_02_l}}
     {targetBone {boneName=dyn_calf_flapB_in_l}}
    ]
    constraintWeights [1,1,1,1,1]
   }
   n5 = "Constraint - dyn_parent_calf_flapA_in_r (PPPPP)"!(
    ComponentPose=@n4
    Alpha=1
    ConstraintWeights_0=0.0089
    ConstraintWeights_1:float=0.4753
    ConstraintWeights_2:float=0.2606
    ConstraintWeights_3:float=0.0041
    ConstraintWeights_4:float=0.2512
   ) -> (Pose) {
    alphaBoolBlend {blendOption=Linear}
    boneToModify {boneName=dyn_parent_calf_flapA_in_r}
    constraintSetup [
     {targetBone {boneName=thigh_r}}
     {targetBone {boneName=dyn_calf_flapA_in_r}}
     {targetBone {boneName=calf_twist_02_r}}
     {targetBone {boneName=calf_twist_01_r}}
     {targetBone {boneName=calf_r}}
    ]
    constraintWeights [1,1,1,1,1]
   }
   n6 = "Constraint - dyn_parent_calf_flapA_out_r (PPPPP)"!(
    ComponentPose=@n5
    Alpha=1
    ConstraintWeights_0=0.0001
    ConstraintWeights_1:float=0.4205
    ConstraintWeights_2:float=0.0049
    ConstraintWeights_3:float=0.2853
    ConstraintWeights_4:float=0.2891
   ) -> (Pose) {
    alphaBoolBlend {blendOption=Linear}
    boneToModify {boneName=dyn_parent_calf_flapA_out_r}
    constraintSetup [
     {targetBone {boneName=thigh_r}}
     {targetBone {boneName=calf_twist_02_r}}
     {targetBone {boneName=calf_twist_01_r}}
     {targetBone {boneName=dyn_calf_flapA_out_r}}
     {targetBone {boneName=calf_r}}
    ]
    constraintWeights [1,1,1,1,1]
   }
   n7 = "Constraint - dyn_parent_calf_flapB_in_r (PPPP)"!(
    ComponentPose=@n6
    Alpha=1
    ConstraintWeights_0=0.4672
    ConstraintWeights_1:float=0.3462
    ConstraintWeights_2:float=0.1595
    ConstraintWeights_3:float=0.0271
   ) -> (Pose) {
    alphaBoolBlend {blendOption=Linear}
    boneToModify {boneName=dyn_parent_calf_flapB_in_r}
    constraintSetup [
     {targetBone {boneName=dyn_calf_flapB_in_r}}
     {targetBone {boneName=calf_twist_02_r}}
     {targetBone {boneName=calf_twist_01_r}}
     {targetBone {boneName=calf_r}}
    ]
    constraintWeights [1,1,1,1]
   }
   "Constraint - dyn_parent_calf_flapB_out_r (PPPP)"!(
    ComponentPose=@n7
    Alpha=1
    ConstraintWeights_0=0.3687
    ConstraintWeights_1:float=0.3556
    ConstraintWeights_2:float=0.2406
    ConstraintWeights_3:float=0.035
   ) -> (Pose) {
    alphaBoolBlend {blendOption=Linear}
    boneToModify {boneName=dyn_parent_calf_flapB_out_r}
    constraintSetup [
     {targetBone {boneName=dyn_calf_flapB_out_r}}
     {targetBone {boneName=calf_twist_02_r}}
     {targetBone {boneName=calf_twist_01_r}}
     {targetBone {boneName=calf_r}}
    ]
    constraintWeights [1,1,1,1]
   }
   n8 = RigidBody(ComponentPose=Local To Component(LocalPose=Use cached pose 'body_pose'())) {
    bClampLinearTranslationLimitToRefPose=true
    bOverrideWorldGravity=true
    overridePhysicsAsset=&/M_MED_Hardcore_Sportz_Physics
    simulationSpace=BaseBoneSpace
    alphaBoolBlend {blendOption=Linear}
    baseBoneRef {boneName=root}
    componentAppliedLinearAccClamp=<1000 1000 1000>
    componentLinearAccScale=<0.3 0.3 0.35>
    componentLinearVelScale=<0.3 0.3 0.35>
    overrideWorldGravity {z=-980}
   }
   n9 = RigidBody(ComponentPose=Local To Component(LocalPose=Use cached pose 'body_pose'()), Alpha=0.6) {
    bOverrideWorldGravity=true
    overridePhysicsAsset=&M_MED_Hardcore_Sportz_Physics
    simulationSpace=BaseBoneSpace
    alphaBoolBlend {blendOption=Linear}
    baseBoneRef {boneName=spine_04}
    componentAppliedLinearAccClamp=<700 700 700>
    componentLinearAccScale=<0.2 0.2 0.2>
    componentLinearVelScale=<0.2 0.2 0.2>
    overrideWorldGravity {z=-980}
   }
   n10 = Get bIsMoving2D()
   n11 = Get bIsAccelerating2D()
   Output Pose(Result=Two Way Blend!(
     A=Component To Local(ComponentPose=@n8)
     B=Component To Local(ComponentPose=@n9)
     bAlphaBoolEnabled:bool=OR Boolean!(A=@n10, B=@n11, C:bool) -> (ReturnValue)
    ) -> (Pose) {blendNode {alphaInputType=Bool, alphaBoolBlend {blendOption=Linear}}}) {layerGroup=None}
   n12 = "Transform (Modify) Bone - Bone: dyn_driver_shldrpad_l"(
    ComponentPose=Local To Component(LocalPose=Copy Pose From Mesh() {bUseAttachedParent=true})
    Alpha=1
    Translation="3,0,0"
    Rotation="5,5,-20"
    Scale="1,1,1"
   ) {
    rotationMode=BMM_Additive
    rotationSpace=BCS_BoneSpace
    scaleSpace=BCS_BoneSpace
    translationMode=BMM_Additive
    translationSpace=BCS_BoneSpace
    alphaBoolBlend {blendOption=Linear}
    alphaScaleBiasClamp {bInterpResult=true, interpSpeedDecreasing=25, interpSpeedIncreasing=35}
    boneToModify {boneName=dyn_driver_shldrpad_l}
    rotation=<rot 5 5 -20>
    translation {x=3}
   }
   n13 = "Transform (Modify) Bone - Bone: dyn_driver_shldrpad_l"(ComponentPose=@n12, Alpha=Get arm_l_fwd_alpha() -> (arm_l_fwd_alpha:double), Translation="0,0,0", Rotation="0,0,30", Scale="1,1,1") {
    rotationMode=BMM_Additive
    rotationSpace=BCS_BoneSpace
    scaleSpace=BCS_BoneSpace
    translationSpace=BCS_BoneSpace
    alphaBoolBlend {blendOption=Linear}
    alphaScaleBiasClamp {bInterpResult=true, interpSpeedDecreasing=25, interpSpeedIncreasing=35}
    boneToModify {boneName=dyn_driver_shldrpad_l}
    rotation {roll=30}
   }
   n14 = Get shldr_r_alpha() -> (shldr_r_alpha:double)
   n15 = "Transform (Modify) Bone - Bone: dyn_driver_shldrpad_l"(ComponentPose=@n13, Alpha=@n14, Translation="0.205,1.078,-0.261", Rotation="9,-5,-15", Scale="0.9,0.9,0.9") {
    rotationMode=BMM_Additive
    rotationSpace=BCS_BoneSpace
    scaleMode=BMM_Additive
    scaleSpace=BCS_BoneSpace
    translationMode=BMM_Additive
    translationSpace=BCS_BoneSpace
    alphaBoolBlend {blendOption=Linear}
    alphaScaleBiasClamp {bInterpResult=true, interpSpeedDecreasing=25, interpSpeedIncreasing=25}
    boneToModify {boneName=dyn_driver_shldrpad_l}
    rotation=<rot 9 -5 -15>
    scale=<0.9 0.9 0.9>
    translation=<0.205 1.078 -0.261>
   }
   n16 = "Transform (Modify) Bone - Bone: dyn_shldrpad_01_l"(ComponentPose=@n15, Alpha=@n14, Translation="-2.571,0.210,0.030", Rotation="0.930,-7.352,0.000", Scale="1,1,1") {
    rotationMode=BMM_Additive
    rotationSpace=BCS_BoneSpace
    scaleMode=BMM_Additive
    scaleSpace=BCS_BoneSpace
    translationMode=BMM_Additive
    translationSpace=BCS_BoneSpace
    alphaBoolBlend {blendOption=Linear}
    alphaScaleBiasClamp {bInterpResult=true, interpSpeedDecreasing=25, interpSpeedIncreasing=25}
    boneToModify {boneName=dyn_shldrpad_01_l}
    rotation {pitch=0.93, yaw=-7.3522}
    translation=<-2.571 0.21 0.03>
   }
   n17 = "Transform (Modify) Bone - Bone: dyn_driver_shldrpad_l"(
    ComponentPose=@n16
    Alpha=Get shldr_r_roll_alpha() -> (shldr_r_roll_alpha:double)
    Translation="12,0,2"
    Rotation="0,5,10"
    Scale="0.9,0.9,0.9"
   ) {
    rotationMode=BMM_Additive
    rotationSpace=BCS_BoneSpace
    scaleMode=BMM_Additive
    scaleSpace=BCS_BoneSpace
    translationMode=BMM_Additive
    translationSpace=BCS_BoneSpace
    alphaBoolBlend {blendOption=Linear}
    alphaScaleBiasClamp {bInterpResult=true, interpSpeedDecreasing=50, interpSpeedIncreasing=50}
    boneToModify {boneName=dyn_driver_shldrpad_l}
    rotation {roll=10, yaw=5}
    scale=<0.9 0.9 0.9>
    translation {x=12, z=2}
   }
   n18 = "Transform (Modify) Bone - Bone: dyn_driver_shldrpad_l"(
    ComponentPose=@n17
    Alpha=Get shldr_l_roll_fwd_alpha() -> (shldr_l_roll_fwd_alpha:double)
    Translation="0,0,0"
    Rotation="0.000,0.000,-30"
    Scale="1,1,1"
   ) {
    rotationMode=BMM_Additive
    rotationSpace=BCS_BoneSpace
    scaleMode=BMM_Additive
    scaleSpace=BCS_BoneSpace
    translationMode=BMM_Additive
    translationSpace=BCS_BoneSpace
    alphaBoolBlend {blendOption=Linear}
    alphaScaleBiasClamp {bInterpResult=true, interpSpeedDecreasing=30, interpSpeedIncreasing=30}
    boneToModify {boneName=dyn_driver_shldrpad_l}
    rotation {roll=-30}
   }
   n19 = Get knee_bend_l_alpha() -> (knee_bend_l_alpha:double)
   n20 = "Transform (Modify) Bone - Bone: dyn_knee_l"(ComponentPose=@n18, Alpha=@n19, Translation="1,-2,-0.000", Rotation="0,35,0", Scale="0.9,0.9,0.9") {
    rotationMode=BMM_Additive
    rotationSpace=BCS_BoneSpace
    scaleMode=BMM_Additive
    scaleSpace=BCS_BoneSpace
    translationMode=BMM_Additive
    translationSpace=BCS_BoneSpace
    alphaBoolBlend {blendOption=Linear}
    alphaScaleBiasClamp {bInterpResult=true, interpSpeedDecreasing=35, interpSpeedIncreasing=35}
    boneToModify {boneName=dyn_knee_l}
    rotation {yaw=35}
    scale=<0.9 0.9 0.9>
    translation=<1 -2>
   }
   n21 = "Transform (Modify) Bone - Bone: dyn_calf_squish_l"(ComponentPose=@n20, Alpha=@n19, Translation="0,0,0", Rotation="0,0,0", Scale="1,0.5,0.5") {
    rotationMode=BMM_Additive
    rotationSpace=BCS_BoneSpace
    scaleMode=BMM_Additive
    scaleSpace=BCS_BoneSpace
    translationMode=BMM_Additive
    translationSpace=BCS_BoneSpace
    alphaBoolBlend {blendOption=Linear}
    alphaScaleBiasClamp {bInterpResult=true, interpSpeedDecreasing=35, interpSpeedIncreasing=35}
    boneToModify {boneName=dyn_calf_squish_l}
    scale {y=0.5, z=0.5}
   }
   n22 = Get knee_bend_r_alpha() -> (knee_bend_r_alpha:double)
   n23 = "Transform (Modify) Bone - Bone: dyn_knee_r"(ComponentPose=@n21, Alpha=@n22, Translation="-1,2,0.000", Rotation="0,35,0", Scale="0.95,0.95,0.95") {
    rotationMode=BMM_Additive
    rotationSpace=BCS_BoneSpace
    scaleMode=BMM_Additive
    scaleSpace=BCS_BoneSpace
    translationMode=BMM_Additive
    translationSpace=BCS_BoneSpace
    alphaBoolBlend {blendOption=Linear}
    alphaScaleBiasClamp {bInterpResult=true, interpSpeedDecreasing=35, interpSpeedIncreasing=35}
    boneToModify {boneName=dyn_knee_r}
    rotation {yaw=35}
    scale=<0.95 0.95 0.95>
    translation=<-1 2>
   }
   n24 = "Transform (Modify) Bone - Bone: dyn_calf_squish_r"(ComponentPose=@n23, Alpha=@n22, Translation="0,0,0", Rotation="0,0,0", Scale="1,0.5,0.5") {
    rotationMode=BMM_Additive
    rotationSpace=BCS_BoneSpace
    scaleMode=BMM_Additive
    scaleSpace=BCS_BoneSpace
    translationMode=BMM_Additive
    translationSpace=BCS_BoneSpace
    alphaBoolBlend {blendOption=Linear}
    alphaScaleBiasClamp {bInterpResult=true, interpSpeedDecreasing=35, interpSpeedIncreasing=35}
    boneToModify {boneName=dyn_calf_squish_r}
    scale {y=0.5, z=0.5}
   }
   n25 = "Constraint - dyn_parent_knee_strap_in_l (PPPP)"!(
    ComponentPose=@n24
    Alpha=1
    ConstraintWeights_0=0.1516
    ConstraintWeights_1:float=0.0818
    ConstraintWeights_2:float=0
    ConstraintWeights_3:float=0.7666
   ) -> (Pose) {
    alphaBoolBlend {blendOption=Linear}
    boneToModify {boneName=dyn_parent_knee_strap_in_l}
    constraintSetup [
     {targetBone {boneName=calf_l}}
     {targetBone {boneName=thigh_twist_01_l}}
     {targetBone {boneName=calf_twist_02_l}}
     {targetBone {boneName=thigh_l}}
    ]
    constraintWeights [1,1,1,1]
   }
   n26 = "Constraint - dyn_parent_knee_strap_out_l (PPP)"!(ComponentPose=@n25, Alpha=1, ConstraintWeights_0=0.0639, ConstraintWeights_1:float=0.0637, ConstraintWeights_2:float=0.8725) -> (Pose) {
    alphaBoolBlend {blendOption=Linear}
    boneToModify {boneName=dyn_parent_knee_strap_out_l}
    constraintSetup [{targetBone {boneName=calf_l}}, {targetBone {boneName=thigh_twist_01_l}}, {targetBone {boneName=thigh_l}}]
    constraintWeights [1,1,1]
   }
   n27 = "Constraint - dyn_parent_knee_strap_inB_l (PPPP)"!(
    ComponentPose=@n26
    Alpha=1
    ConstraintWeights_0=0.3875
    ConstraintWeights_1:float=0.0195
    ConstraintWeights_2:float=0.0157
    ConstraintWeights_3:float=0.5774
   ) -> (Pose) {
    alphaBoolBlend {blendOption=Linear}
    boneToModify {boneName=dyn_parent_knee_strap_inB_l}
    constraintSetup [
     {targetBone {boneName=calf_l}}
     {targetBone {boneName=thigh_twist_01_l}}
     {targetBone {boneName=calf_twist_02_l}}
     {targetBone {boneName=thigh_l}}
    ]
    constraintWeights [1,1,1,1]
   }
   n28 = "Constraint - dyn_parent_knee_strap_outB_l (PPPP)"!(
    ComponentPose=@n27
    Alpha=1
    ConstraintWeights_0=0.5782
    ConstraintWeights_1:float=0.0084
    ConstraintWeights_2:float=0.0286
    ConstraintWeights_3:float=0.3848
   ) -> (Pose) {
    alphaBoolBlend {blendOption=Linear}
    boneToModify {boneName=dyn_parent_knee_strap_outB_l}
    constraintSetup [
     {targetBone {boneName=calf_l}}
     {targetBone {boneName=thigh_twist_01_l}}
     {targetBone {boneName=calf_twist_02_l}}
     {targetBone {boneName=thigh_l}}
    ]
    constraintWeights [1,1,1,1]
   }
   n29 = "Constraint - dyn_parent_knee_strap_in_r (PPPP)"!(
    ComponentPose=@n28
    Alpha=1
    ConstraintWeights_0=0.7714
    ConstraintWeights_1:float=0
    ConstraintWeights_2:float=0.0862
    ConstraintWeights_3:float=0.1424
   ) -> (Pose) {
    alphaBoolBlend {blendOption=Linear}
    boneToModify {boneName=dyn_parent_knee_strap_in_r}
    constraintSetup [
     {targetBone {boneName=thigh_r}}
     {targetBone {boneName=calf_twist_02_r}}
     {targetBone {boneName=thigh_twist_01_r}}
     {targetBone {boneName=calf_r}}
    ]
    constraintWeights [1,1,1,1]
   }
   n30 = "Constraint - dyn_parent_knee_strap_out_r (PPP)"!(ComponentPose=@n29, Alpha=1, ConstraintWeights_0=0.8196, ConstraintWeights_1:float=0.035, ConstraintWeights_2:float=0.1454) -> (Pose) {
    alphaBoolBlend {blendOption=Linear}
    boneToModify {boneName=dyn_parent_knee_strap_out_r}
    constraintSetup [{targetBone {boneName=thigh_r}}, {targetBone {boneName=thigh_twist_01_r}}, {targetBone {boneName=calf_r}}]
    constraintWeights [1,1,1]
   }
   n31 = "Constraint - dyn_parent_knee_strap_inB_r (PPPPP)"!(
    ComponentPose=@n30
    Alpha=1
    ConstraintWeights_0=0.3445
    ConstraintWeights_1:float=0.2233
    ConstraintWeights_2:float=0.0312
    ConstraintWeights_3:float=0.01
    ConstraintWeights_4:float=0.3909
   ) -> (Pose) {
    alphaBoolBlend {blendOption=Linear}
    boneToModify {boneName=dyn_parent_knee_strap_inB_r}
    constraintSetup [
     {targetBone {boneName=thigh_r}}
     {targetBone {boneName=dyn_calf_squish_r}}
     {targetBone {boneName=calf_twist_02_r}}
     {targetBone {boneName=thigh_twist_01_r}}
     {targetBone {boneName=calf_r}}
    ]
    constraintWeights [1,1,1,1,1]
   }
   n32 = "Constraint - dyn_parent_knee_strap_outB_r (PPPPP)"!(
    ComponentPose=@n31
    Alpha=1
    ConstraintWeights_0=0.3931
    ConstraintWeights_1:float=0.0974
    ConstraintWeights_2:float=0.0255
    ConstraintWeights_3:float=0.0117
    ConstraintWeights_4:float=0.4722
   ) -> (Pose) {
    alphaBoolBlend {blendOption=Linear}
    boneToModify {boneName=dyn_parent_knee_strap_outB_r}
    constraintSetup [
     {targetBone {boneName=thigh_r}}
     {targetBone {boneName=dyn_calf_squish_r}}
     {targetBone {boneName=calf_twist_02_r}}
     {targetBone {boneName=thigh_twist_01_r}}
     {targetBone {boneName=calf_r}}
    ]
    constraintWeights [1,1,1,1,1]
   }
   Save cached pose 'body_pose'(Pose=Component To Local(ComponentPose=@n32)) {cacheName=body_pose}
   Get shldr_l_alpha() -> (shldr_l_alpha:double)
   Get bIsInFrontEnd()
  ]}
 EventGraph {nodes [
   n1 = Event BlueprintUpdateAnimation()
   > Get!(execute:Exec) -> (then:Exec, else:Exec, Skel_Mesh:Skeletal Mesh Component Object Reference)
   > n2 = Sequence!(execute) -> (then_0, then_1, then_2:Exec, then_3:Exec, then_4:Exec, then_5:Exec)
   n3 = Get Skel_Mesh()
   Set shldr_l_roll_fwd_alpha(
    execute=@n2.then_2
    shldr_l_roll_fwd_alpha:double=GetAlphaFromReferenceTransformDelta(
     SkeletalMeshComponent=@n3
     BoneName=upperarm_l
     BaseName=clavicle_l
     DeltaComponentType=QuaternionTwist
     InputRange=(X=0,Y=60)
     TwistAxis="1,0,0"
    )
   ) -> (Output_Get:double)
   Set shldr_r_alpha(
    execute=@n2.then_1
    shldr_r_alpha:double=GetAlphaFromReferenceTransformDelta(
     SkeletalMeshComponent=@n3
     BoneName=upperarm_l
     BaseName=clavicle_l
     DeltaComponentType=QuaternionTwist
     InputRange=(X=-70,Y=-120)
     TwistAxis="0,0,1"
    )
   ) -> (Output_Get:double)
   > Set shldr_r_pouch_alpha(shldr_r_pouch_alpha:double=GetAlphaFromReferenceTransformDelta(
     SkeletalMeshComponent=@n3
     BoneName=upperarm_l
     BaseName=clavicle_l
     DeltaComponentType=QuaternionTwist
     InputRange=(X=0,Y=-75)
     TwistAxis="0,0,1"
    )) -> (Output_Get:double)
   n4 = "Clamp (Float)"(
    Value=float * float(
     A=GetAlphaFromReferenceTransformDelta(
      SkeletalMeshComponent=@n3
      BoneName=upperarm_l
      BaseName=clavicle_l
      DeltaComponentType=QuaternionTwist
      InputRange=(X=0,Y=-90)
      TwistAxis="1,0,0"
     )
     B=GetAlphaFromReferenceTransformDelta(
      SkeletalMeshComponent=@n3
      BoneName=upperarm_l
      BaseName=clavicle_l
      DeltaComponentType=QuaternionTwist
      InputRange=(X=0,Y=80)
      TwistAxis="0,1,0"
     )
    )
    Min=0
    Max=1
   )
   > n5 = Branch(Condition="float <= float"!(A:double=@n4, B:double=0.1) -> (ReturnValue:bool) {is=title K2Node_CallFunction})
   > Set shldr_r_roll_alpha(shldr_r_roll_alpha:double=@n4) -> (Output_Get:double)
   Set shldr_r_roll_alpha(execute=@n5.else, shldr_r_roll_alpha=FInterpTo(Current=Get shldr_r_roll_alpha(), Target=@n4, DeltaTime=@n1.DeltaTimeX, InterpSpeed=20))
   n6 = Get Skel_Mesh()
   n7 = GetAlphaFromReferenceTransformDelta(
    SkeletalMeshComponent=@n6
    BoneName=upperarm_l
    BaseName=clavicle_l
    DeltaComponentType=QuaternionTwist
    InputRange=(X=0,Y=90)
    TwistAxis="0,1,0"
   )
   n8 = GetAlphaFromReferenceTransformDelta(
    SkeletalMeshComponent=@n6
    BoneName=upperarm_l
    BaseName=clavicle_l
    DeltaComponentType=QuaternionTwist
    InputRange=(X=0,Y=-100)
    TwistAxis="1,0,0"
   )
   Set shldr_l_alpha(execute=@n2.then_0, shldr_l_alpha:double="Clamp (Float)"(Value=float * float(A=@n8, B=@n7), Min=0, Max=1)) -> (Output_Get:double)
   > n9 = Branch(Condition=Get bIsTargeting())
   > Set shldr_l_ads_alpha(shldr_l_ads_alpha:double=1) -> (Output_Get:double)
   Set shldr_l_ads_alpha(execute=@n9.else, shldr_l_ads_alpha=0)
   n10 = Get Skel_Mesh()
   Set knee_bend_l_alpha(
    execute=@n2.then_3
    knee_bend_l_alpha:double=GetAlphaFromReferenceTransformDelta(
     SkeletalMeshComponent=@n10
     BoneName=calf_l
     BaseName=thigh_l
     DeltaComponentType=QuaternionTwist
     InputRange=(X=0,Y=-110)
     TwistAxis="0,0,1"
    )
   ) -> (Output_Get:double)
   Set knee_bend_r_alpha(
    execute=@n2.then_4
    knee_bend_r_alpha:double=GetAlphaFromReferenceTransformDelta(
     SkeletalMeshComponent=@n10
     BoneName=calf_r
     BaseName=thigh_r
     DeltaComponentType=QuaternionTwist
     InputRange=(X=0,Y=-110)
     TwistAxis="0,0,1"
    )
   ) -> (Output_Get:double)
   Set arm_l_fwd_alpha(
    execute=@n2.then_5
    arm_l_fwd_alpha:double=GetAlphaFromReferenceTransformDelta(
     SkeletalMeshComponent=Get Skel_Mesh()
     BoneName=upperarm_l
     BaseName=clavicle_l
     DeltaComponentType=QuaternionTwist
     InputRange=(X=0,Y=-90)
     TwistAxis="1,0,0"
    )
   ) -> (Output_Get:double)
   Event OnBeginCrouching()
   > Set crouch_bool(crouch_bool:bool=true) -> (Output_Get:bool)
   Event OnEndCrouching()
   > Set crouch_bool(crouch_bool=true)
   Event BlueprintInitializeAnimation()
   > Set Skel_Mesh(Skel_Mesh=GetOwnerSkeletalMeshComponent())
  ]}
]
variables [
 arm_l_fwd_alpha {type=number}
 crouch_bool {type=boolean}
 knee_bend_l_alpha {type=number}
 knee_bend_r_alpha {type=number}
 shldr_l_ads_alpha {type=number}
 shldr_l_alpha {type=number}
 shldr_l_roll_fwd_alpha {type=number}
 shldr_r_alpha {type=number}
 shldr_r_pouch_alpha {type=number}
 shldr_r_roll_alpha {type=number}
 Skel_Mesh {type=SkeletalMeshComponent}
]
